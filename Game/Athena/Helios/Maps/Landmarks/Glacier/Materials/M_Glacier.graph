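# Glacier landmark material for the Helios map, building ice from layered cliff height, normal and scratch detail, with world-space normal output for Nanite and instanced meshes.
# A height lerp adds landscape-matched snow by slope, vertex color and noise, while an optional branch blends attribute layers for flow.
# World Z drives waterline darkening and ice glow, then RVT blending follows, with emissive masked by RVT height and tinted by time of day.
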Material /Game/Athena/Helios/Maps/Landmarks/Glacier/Materials/M_Glacier
graph {
 properties {
  bTangentSpaceNormal=false
  bUseMaterialAttributes=true
  bUsedWithInstancedStaticMeshes=true
  bUsedWithNanite=true
  pixelDepthOffsetMode=PDOM_Legacy
  usedShadingModels=Default Lit
 }
 nodes [
  n1 = MaterialFunctionCall(
   Transition Phase=LinearInterpolate(A=Constant() {r=0.3}, B=Constant() {r=1}, Alpha=NamedRerouteUsage())
   Height Texture=LinearInterpolate(A=Constant() {r=-0.5}, B=Constant() {r=1.5}, Alpha=NamedRerouteUsage())
  ) {materialFunction=&HeightLerp}
  Multiply(A=@n1.Lerp Alpha No Contrast)
  n2 = MaterialFunctionCall!(Grass_UseCustomSlopeMask, Grass_CustomSlopeMask) {materialFunction=/Game/Athena/Helios/Environments/Landscape/Materials/MaterialFunctions_Utilities/MF_Helios_StaticMeshLandscapeLayerMatch}
  n3 = GetMaterialAttributes(None=@n2.SnowLayer) {attributeGetTypes [#1, #2]}
  NamedRerouteDeclaration(None=@n3.Opacity) {name=Height Snow, nodeColor=<rgba 1 0.75 0 1>}
  n4 = TextureSample(UVs=Multiply(A=TextureCoordinate(), B=ScalarParameter() {defaultValue=0.5, parameterName=Tiling_Scratches})) {samplerType=SAMPLERTYPE_Normal, texture=/Game/Global/Textures/Default/Tileable_Patterns/T_IceScratch1_N}
  NamedRerouteDeclaration(None=@n4.R) {name=Mask_Scratches, nodeColor=<rgba 0 1 0.1562 1>}
  n5 = VectorParameter() {group=Ice, parameterName=Ice Color_Top, defaultValue=<rgba 0.1352 0.1908 0.2066 2>}
  n6 = VectorParameter() {group=Ice, parameterName=Ice Color, defaultValue=<rgba 0.1352 0.1908 0.2066 2>}
  n7 = NamedRerouteUsage()
  n8 = Multiply(
   A=ScalarParameter() {defaultValue=0.5, parameterName=Add Details}
   B=Multiply(A=NamedRerouteUsage(), B=Constant3Vector() {constant=<rgba 1 1 0 0>})
  )
  n9 = Multiply(B=@n8) {constA=0.25}
  n10 = Saturate(None=LinearInterpolate(A=Constant() {r=-3}, B=Constant() {r=2}, Alpha=NamedRerouteUsage()))
  n11 = NamedRerouteUsage()
  n12 = Multiply(A=Multiply(A=LinearInterpolate(A=@n9, B=@n8, Alpha=@n10), B=LinearInterpolate(Alpha=@n11) {constA=1, constB=0.25}), B=NamedRerouteUsage())
  n13 = NamedRerouteUsage()
  n14 = VectorParameter() {group=IceEmissive, parameterName=Emissive color Ice, defaultValue=<rgba 0.0317 0.1465 0.15 0.5>}
  n15 = Max(A=@n14.A, B=NamedRerouteUsage())
  n16 = WorldPosition()
  n17 = ScalarParameter() {defaultValue=4200, group=IceEmissive, parameterName=GlowLine}
  n18 = Saturate(None=Divide(A=Subtract(A=@n16.Z, B=@n17), B=ScalarParameter() {defaultValue=2500, group=IceEmissive, parameterName=GlowFade}))
  n19 = ScalarParameter() {defaultValue=0.5, group=Ice, parameterName=Ice Roughness}
  n20 = SetMaterialAttributes!(
   MaterialAttributes
   Base Color=LinearInterpolate(A=LinearInterpolate(A=@n5.RGB, B=@n6.RGB, Alpha=@n7), B=Multiply(A=@n6.RGB, B=@n6.RGB), Alpha=NamedRerouteUsage())
   Normal=QualitySwitch(Default=Add(A=@n12, B=@n13), Low=@n13)
   Roughness=@n19
   Emissive Color=Multiply(A=Multiply(A=@n14.RGB, B=@n15), B=NamedRerouteDeclaration(None=@n18) {name=GlowGradient, nodeColor=<rgba 0.0382 0.3018 1 1>})
  ) {attributeSetTypes [
    #3
    #2
    #4
    #5
   ]}
  NamedRerouteDeclaration(None=@n20) {name=MA_BaseGlacier, nodeColor=<rgba 1 0 0.1641 1>}
  n21 = VectorParameter() {group=Water, parameterName=FlowColor, defaultValue=<rgba 0.0879 0.1454 0.1649 1>}
  n22 = TextureCoordinate()
  n23 = ScalarParameter() {defaultValue=2, group=Water, parameterName=Tiling_Water}
  n24 = Constant()
  n25 = Constant() {r=-0.1}
  n26 = TextureSample(UVs=Panner(Coordinate=Multiply(A=@n22, B=@n23), Speed=AppendVector(A=@n24, B=@n25))) {samplerType=SAMPLERTYPE_Normal, texture=/Game/Global/Textures/Default/Tileable_Patterns/T_Water2_N}
  n27 = Multiply(A=Multiply(A=@n26.G) {constB=0.15}, B=@n21.RGB)
  n28 = Constant3Vector() {constant=<rgba 0.2 0.2 1 0>}
  NamedRerouteDeclaration(None=SetMaterialAttributes!(
    MaterialAttributes=NamedRerouteUsage()
    Base Color=Add(A=@n21.RGB, B=@n27)
    Normal=Multiply(A=@n26.RGB, B=@n28)
    Roughness=ScalarParameter() {group=Water, parameterName=Water_Roughness}
   ) {attributeSetTypes [#3, #2, #4]}) {name=MA_WetGlacier, nodeColor=<rgba 1 0 0.7266 1>}
  n29 = GetMaterialAttributes(None=NamedRerouteUsage()) {attributeGetTypes [#2]}
  n30 = Add(
   A=@n29.Normal
   B=Multiply(
    A=TextureSample() {samplerType=SAMPLERTYPE_Normal, texture=&T_Water2_N}.RGB
    B=Constant3Vector() {constant=<rgba 2 2 0 0>}
   )
  )
  n31 = Constant() {r=0.75}
  n32 = CollectionParameter() {
   collection=&FortniteMaterialParameters
   parameterName=FortniteActiveTimeOfDay
  }
  n33 = DotProduct(
   A=MaterialFunctionCall(X=@n31, Y=Constant() {r=1}, Z=@n31, A=Constant()) {materialFunction=&MakeFloat4}.Result
   B=@n32
  )
  NamedRerouteDeclaration(None=BlendMaterialAttributes(
    A=SetMaterialAttributes!(MaterialAttributes=@n29.MaterialAttributes, Normal=@n30) {attributeSetTypes [#2]}
    B=NamedRerouteUsage()
    Alpha=Saturate(None=@n33)
   )) {name=FrozenGlacier, nodeColor=<rgba 0.5 1 0 1>}
  n34 = ComponentMask(None=TransformPosition(None=WorldPosition().XYZ) {transformSourceType=TRANSFORMPOSSOURCE_World}) {g=true}
  NamedRerouteDeclaration(None=SmoothStep(
    Min=ScalarParameter() {defaultValue=8000, parameterName=smoothmin}
    Max=ScalarParameter() {defaultValue=9000, parameterName=smoothmax}
    Value=Multiply(A=@n34) {constB=-1}
   )) {name=FrontGradient, nodeColor=<rgba 1 0.8672 0 1>}
  n35 = TextureCoordinate()
  NamedRerouteDeclaration(None=SmoothStep(Value=ComponentMask(None=@n35) {g=true}) {constMax=-0.9, constMin=-1.2}) {name=Mask_Front, nodeColor=<rgba 1 0 0.1406 1>}
  n36 = Multiply(A=TextureCoordinate(), B=ScalarParameter() {defaultValue=1, group=Ice, parameterName=Tiling_Base})
  n37 = TextureSample(UVs=@n36) {
   samplerSource=SSM_Wrap_WorldGroupSettings
   samplerType=SAMPLERTYPE_LinearGrayscale
   texture=/Game/Athena/Helios/Maps/Landmarks/Glacier/Textures/T_Artemis_Cliff_01_H
  }
  n38 = Multiply(A=TextureCoordinate() {coordinateIndex=1}, B=ScalarParameter() {defaultValue=5, group=Ice, parameterName=Tiling_Detail}) {constB=5}
  n39 = TextureSample(UVs=@n38) {
   samplerSource=SSM_Wrap_WorldGroupSettings
   samplerType=SAMPLERTYPE_LinearGrayscale
   texture=&T_Artemis_Cliff_01_H
  }
  NamedRerouteDeclaration(None=Max(
    A=NamedRerouteDeclaration(None=@n37.R) {name=GlacierHeight_Base, nodeColor=<rgba 0 0.8594 1 1>}
    B=NamedRerouteDeclaration(None=@n39.R) {name=GlacierHeight_Detail, nodeColor=<rgba 0 0.2266 1 1>}
   )) {name=GlacierHeight_Combined, nodeColor=<rgba 0 0.7656 1 1>}
  n40 = TextureSampleParameter2D(UVs=@n36) {
   group=Material_1
   parameterName=Normals
   samplerSource=SSM_Wrap_WorldGroupSettings
   samplerType=SAMPLERTYPE_Normal
   sortPriority=-5
   texture=/Game/Environments/Artemis/Sets/Terrain/Textures/T_Artemis_Cliff_01_N
  }
  n41 = Multiply(
   A=NamedRerouteDeclaration(None=@n40.RGB) {name=GlacierNormals_Base, nodeColor=<rgba 0.3828 0 1 1>}
   B=Constant3Vector() {constant=<rgba 1 1 0 0>}
  )
  n42 = ScalarParameter() {defaultValue=1, group=Ice, parameterName=Normals}
  n43 = TextureSample(UVs=@n38) {
   samplerSource=SSM_Wrap_WorldGroupSettings
   samplerType=SAMPLERTYPE_Normal
   texture=&T_Artemis_Cliff_01_N
  }
  n44 = MaterialFunctionCall!(UVs=@n38, Normal=@n43.RGB) {materialFunction=/Engine/Functions/Engine_MaterialFunctions01/Texturing/DeriveTangentBasis}
  NamedRerouteDeclaration(None=Add(
    A=Multiply(A=@n41, B=@n42)
    B=Transform(None=@n44.World Space Normal) {transformSourceType=TRANSFORMSOURCE_World, transformType=TRANSFORM_Tangent}
   )) {name=GlacierNormals_Combined, nodeColor=<rgba 1 0 0.1641 1>}
  NamedRerouteDeclaration(None=@n4.RGB) {name=Normals_Scratches, nodeColor=<rgba 1 0.8438 0 1>}
  n45 = NamedRerouteUsage()
  n46 = NamedRerouteUsage()
  QualitySwitch(Default=Multiply(A=@n19, B=LinearInterpolate(B=@n45, Alpha=@n46) {constA=1}), Low=@n19) {desc=removed to address FORT-692350}
  n47 = BlendMaterialAttributes(A=NamedRerouteUsage(), B=NamedRerouteUsage(), Alpha=@n1.Lerp Alpha No Contrast)
  n48 = NamedRerouteUsage()
  n49 = GetMaterialAttributes(None=QualitySwitch(Default=StaticSwitchParameter(True=@n47, False=@n48) {parameterName=Flow True}, Low=@n48, Medium=@n48)) {attributeGetTypes [#2]}
  n50 = SetMaterialAttributes(MaterialAttributes=@n49.MaterialAttributes, Normal=Transform(None=@n49.Normal)) {attributeSetTypes [#2]}
  n51 = Transform(None=Constant3Vector() {constant=<rgba 0 0 1 0>})
  n52 = VertexNormalWS()
  n53 = NamedRerouteDeclaration(None=Saturate(None=ComponentMask(None=@n52) {b=true})) {name=Up Simple, nodeColor=<rgba 0.3359 0 1 1>}
  n54 = LinearInterpolate(A=@n53, B=Multiply(A=@n53, B=ScalarParameter() {defaultValue=1, parameterName=FrontUpNormalInfluence}), Alpha=NamedRerouteUsage())
  n55 = GetMaterialAttributes(None=SetMaterialAttributes(MaterialAttributes=@n3.MaterialAttributes, Normal=LinearInterpolate(A=@n51, B=@n3.Normal, Alpha=@n54)) {attributeSetTypes [#2]}) {attributeGetTypes [#3]}
  n56 = MaterialProxyReplace(
   Realtime=@n55.Base Color
   MaterialProxy=VectorParameter() {parameterName=MaterialProxyDiffuse, defaultValue=<rgba 0.2747 0.3864 0.4735 1>}.RGB
  )
  n57 = NamedRerouteUsage()
  n58 = ScalarParameter() {parameterName=IceBlend_BaseHeightInfluence}
  n59 = NamedRerouteUsage()
  n60 = ScalarParameter() {parameterName=IceBlend_CombinedHeightInfluence}
  n61 = Add(
   A=Subtract(A=ScalarParameter() {defaultValue=0.6, parameterName=Snow To Ice Blend}, B=Multiply(A=@n57, B=@n58))
   B=Max(A=Multiply(A=@n59, B=@n60), B=VertexColor().B)
  )
  n62 = Max(
   A=VertexColor().G
   B=SphereMask(
    A=VertexNormalWS()
    B=Constant3Vector() {constant=<rgba 0 0 1 0>}
    Radius=Max(A=@n61) {constB=0}
    Hardness=ScalarParameter() {defaultValue=0.35, parameterName=Snow To Ice Hardness}
   )
  )
  n63 = NamedRerouteUsage()
  n64 = Multiply(A=@n63, B=ScalarParameter() {defaultValue=1.5, parameterName=SnowIce HM Value})
  n65 = TextureSample(UVs=Multiply(A=TextureCoordinate(), B=ScalarParameter() {defaultValue=2, group=Ice, parameterName=Tiling_SnowNoise})) {samplerType=SAMPLERTYPE_Grayscale, texture=/Game/Environments/Apollo/Sets/Terrain/Textures/T_Apollo_SnowMask}
  n66 = ComponentMask(None=Multiply(A=OneMinus(None=@n65.R), B=NamedRerouteUsage())) {a=true}
  n67 = Subtract(A=@n64, B=Multiply(A=@n64, B=@n66))
  n68 = ScalarParameter() {defaultValue=1, parameterName=SnowIce Contrast}
  n69 = OneMinus(None=@n63)
  n70 = Multiply(A=Add(A=Multiply(A=@n68, B=@n69), B=@n66), B=OneMinus(None=VertexColor().B))
  n71 = StaticSwitchParameter(
   True=BlendMaterialAttributes(
    A=@n50
    B=SetMaterialAttributes!(MaterialAttributes=@n55.MaterialAttributes, Base Color=@n56) {attributeSetTypes [#3]}
    Alpha=MaterialFunctionCall(Transition Phase=@n62, Height Texture=@n67, Contrast=@n70) {materialFunction=&HeightLerp}.Alpha
   ) {vertexAttributeBlendType=UseA}
   False=@n50
  ) {defaultValue=true, parameterName=Snow Top}
  n72 = GetMaterialAttributes(None=@n71) {attributeGetTypes [#3, #6, #4]}
  n73 = VectorParameter() {group=Water, parameterName=Water, defaultValue=<rgba 0.1104 0.1802 0.2483 0.5>}
  n74 = Multiply(A=@n72.Base Color, B=@n73.RGB)
  n75 = Multiply(A=@n72.Specular, B=@n73.A)
  n76 = Multiply(A=@n72.Roughness, B=ScalarParameter() {defaultValue=0.5, group=Water, parameterName=RWater})
  n77 = Divide(
   A=Subtract(A=WorldPosition().Z, B=ScalarParameter() {defaultValue=4200, group=Water, parameterName=Waterline Z})
   B=ScalarParameter() {defaultValue=2000, group=Water, parameterName=Waterline Fade}
  )
  n78 = MaterialFunctionCall!(
   In=BlendMaterialAttributes(
    A=SetMaterialAttributes!(MaterialAttributes=@n72.MaterialAttributes, Base Color=@n74, Specular=@n75, Roughness=@n76) {attributeSetTypes [#3, #6, #4]}
    B=@n71
    Alpha=Saturate(None=@n77)
   ) {vertexAttributeBlendType=UseA}
   "RVT Height Map (Roughness)"=NamedRerouteUsage()
  ) {desc="RVT ", materialFunction=/Game/Global/Base4Layers_2022/MaterialFunctions/MF_RVTBlendingRocks}
  n79 = GetMaterialAttributes(None=@n78.Result) {attributeGetTypes [#5]}
  n80 = Divide(A=@n78.RVT H, B=ScalarParameter() {defaultValue=1500, group=A_Emissive_TOD, parameterName=Emissive RVT Mask})
  n81 = Multiply(A=@n79.Emissive Color, B=NamedRerouteDeclaration(None=Saturate(None=@n80)) {name=RVT mask, nodeColor=<rgba 0 0.1797 1 1>})
  n82 = VectorParameter() {group=IceEmissive, parameterName=EmissiveMorning, sortPriority=1, defaultValue=<rgba 1.4513 0.7831 2 1>}
  n83 = VectorParameter() {group=IceEmissive, parameterName=EmissiveDay, sortPriority=2, defaultValue=<rgba 1.3403 1.666 2 1>}
  n84 = VectorParameter() {group=IceEmissive, parameterName=EmissiveEvening, sortPriority=3, defaultValue=<rgba 0.0877 0.2046 0.25 1>}
  n85 = VectorParameter() {group=IceEmissive, parameterName=EmissiveNight, sortPriority=4, defaultValue=<rgba 0.1021 0.1648 0.3 1>}
  n86 = MaterialFunctionCall(Color_Morning=@n82.RGB, Color_Day=@n83.RGB, Color_Evening=@n84.RGB, Color_Night=@n85.RGB) {materialFunction=&Color_Time_of_Day}
  n87 = Desaturation(None=Multiply(A=@n79.Emissive Color, B=@n86.Result), Fraction=Constant() {r=0.5})
  n88 = Max(A=NamedRerouteUsage()) {constB=0.25}
  SetMaterialAttributes!(MaterialAttributes=@n79.MaterialAttributes, Emissive Color=Add(A=Multiply(A=@n81, B=@n86.Result), B=Multiply(A=@n87, B=@n88))) {attributeSetTypes [#5]}
 ]
}
